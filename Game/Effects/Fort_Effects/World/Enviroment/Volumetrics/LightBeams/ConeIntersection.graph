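# Material function that analytically intersects a ray with a Z-capped cone, apparently for volumetric light beams, despite its description saying cylinder.
# Ray origin and direction move into a scaled cone-axis basis, where custom HLSL solves the quadratic and clamps to caps and scene depth.
# Only the abs-discriminant NaN-safe node feeds outputs: thickness, hit distances, local entry and "world" entry/exit positions, which are actually relative to cone origin.
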MaterialFunction /Game/Effects/Fort_Effects/World/Enviroment/Volumetrics/LightBeams/ConeIntersection
graph {
 properties {
  bExposeToLibrary=true
  description=Returns an analytical intersection for a cylinder of the specified coordinates.
  libraryCategoriesText [Math]
 }
 nodes [
  n1 = FunctionInput(Preview=Multiply(A=CameraVectorWS()) {constB=-1}) {
   bUsePreviewValueAsDefault=true
   description="The vector to calculate intersection along. Defaults to negative cameravector, or a normalized vector pointing from the camera to the world."
   inputName=Ray Direction
   sortPriority=1
  }
  n2 = FunctionInput(Preview=Transform(None=Constant3Vector() {constant=<rgba 0 0 1 0>}) {transformSourceType=TRANSFORMSOURCE_Local}) {
   bUsePreviewValueAsDefault=true
   description=The Z axis is the axis along the length of the cylinder. Defaults to the local Z axis of the mesh the material is applied to.
   inputName=Cone Z Axis
   sortPriority=2
  }
  n3 = MaterialFunctionCall!(Vector1=Normalize(VectorInput=@n2), Vector2=VectorParameter() {parameterName=Default Value, defaultValue=<rgba 0 0 1 1>}.RGB) {materialFunction=/Engine/Functions/Engine_MaterialFunctions02/Math/CreateThirdOrthogonalVector}
  n4 = FunctionInput(Preview=FunctionOutput(None=@n3.Vector2) {outputName=CUV Vector 1, sortPriority=20}) {bUsePreviewValueAsDefault=true, inputName=CUV Vector 1, sortPriority=11}
  n5 = FunctionInput(Preview=FunctionOutput(None=@n3.Vector3) {outputName=CUV Vector 2, sortPriority=21}) {bUsePreviewValueAsDefault=true, inputName=CUV Vector 2, sortPriority=12}
  n6 = FunctionInput() {
   bUsePreviewValueAsDefault=true
   description="The X axis is the axis used to apply non uniform scale along.  Defaults to the local X axis of the mesh the material is applied to."
   inputName=XYZ Scale
   sortPriority=4
   previewValue=<1 1 0 1>
  }
  MaterialExpressionMultiply_240 = Multiply(
   A=FunctionOutput(None=MaterialFunctionCall!(VectorToTransform=@n1, BasisX=@n4, BasisY=@n5, BasisZ=@n3.Vector1) {desc=Directional Vec, materialFunction=/Engine/Functions/Engine_MaterialFunctions02/Math/InverseTransformMatrix}.Transformed Vector 3) {outputName=Local Ray Direction, sortPriority=7}
   B=@n6
  ) {desc=D}
  n7 = FunctionInput(Preview=CameraPositionWS()) {bUsePreviewValueAsDefault=true, description=Where to calculate the ray from. Defaults to Camera Position., inputName=Ray Origin}
  n8 = FunctionInput(Preview=ActorPositionWS()) {bUsePreviewValueAsDefault=true, description=The center of the cylinder. Defaults to Actor Position., inputName=Cone Origin, sortPriority=1}
  n9 = Subtract(A=@n7, B=@n8)
  n10 = MaterialFunctionCall(VectorToTransform=@n9, BasisX=@n4, BasisY=@n5, BasisZ=@n3.Vector1) {desc=Origin Vec, materialFunction=&InverseTransformMatrix}
  n11 = Constant3Vector() {constant=<rgba 0 0 1 0>}
  n12 = FunctionInput() {
   bUsePreviewValueAsDefault=true
   description="The X axis is the axis used to apply non uniform scale along.  Defaults to the local X axis of the mesh the material is applied to."
   inputName=Z Min-Max
   inputType=FunctionInput_Vector2
   sortPriority=3
   previewValue=<0 100 0 1>
  }
  n13 = MaterialFunctionCall(Float2=@n12) {materialFunction=&BreakOutFloat2Components}
  MaterialExpressionMultiply_241 = Multiply(A=@n6, B=Add(A=FunctionOutput(None=@n10.Transformed Vector 3) {outputName=Local Ray Origin, sortPriority=6}, B=Multiply(A=@n11, B=@n13.R))) {desc=E}
  MaterialExpressionMultiply_50 = Multiply(A=@n13.R)
  MaterialExpressionMultiply_51 = Multiply(A=@n13.G)
  n14 = DotProduct(
   A=CameraVectorWS()
   B=MaterialFunctionCall() {materialFunction=&CameraDirectionVector}.Result
  )
  MaterialExpressionDivide_9 = Divide(A=SceneDepth().R, B=Max(A=Abs(None=@n14)) {constB=0.0001})
  Custom!(
   D=@MaterialExpressionMultiply_240
   O=@MaterialExpressionMultiply_241
   Min=@MaterialExpressionMultiply_50
   Max=@MaterialExpressionMultiply_51
   SceneDepth=@MaterialExpressionDivide_9
  ) {
   code="float a =  dot(D.xy, D.xy) - D.z * D.z;\r\nfloat b = (dot(O.xy, D.xy) - O.z * D.z) * 2;\r\nfloat c =  dot(O.xy, O.xy) - O.z * O.z;\r\n\r\nfloat d = (b*b) - (4*a*c);\r\nfloat t   = sqrt( d );\r\n\r\nfloat t0 = -b - t;\r\nfloat t1 = -b + t;\r\n\r\nt0 /= 2 * a;\r\nt1 /= 2 * a;\r\n\r\nfloat dist_to_top = max( 0, (Max - O.z) / D.z );\r\nfloat dist_to_bot = max( 0, (Min - O.z) / D.z );\r\n\r\nfloat furthestcap   = dist_to_top > dist_to_bot? Max: Min;\r\nfloat furthestcapXY = length( O.xy + max(dist_to_top, dist_to_bot) * D.xy );\r\n\r\n\r\nfloat entrytest = O.z + t0 * D.z;\r\nfloat exittest =  O.z + t1 * D.z;\r\n\r\nt0 = max( 0, t0);\r\nfloat t0_nocap = t0;\r\n\r\nif( entrytest < Min || entrytest > Max )\r\n{\r\n\tt0 = min(dist_to_top, dist_to_bot);\r\n}\r\n\r\nif( exittest  < Min || exittest  > Max )\r\n{\r\n\tt1 = max(dist_to_top, dist_to_bot);\r\n\t\r\n\tif(furthestcapXY > furthestcap ) t1 = 0;\r\n}\r\n\r\nt1 = min( t1, SceneDepth );\r\n\r\nif( d < 0) t1 = 0;\r\nfloat thickness = (t1 - t0);\r\n\r\nreturn float4(thickness, t0, t1, t0_nocap);\r\n\r\n\r\n"
   description=Cone Intersection
   outputType=CMOT_Float4
   inputs [
    {
     inputName=D
     input {expression=&self:MaterialExpressionMultiply_240}
    }
    {
     inputName=O
     input {expression=&self:MaterialExpressionMultiply_241}
    }
    {
     inputName=Min
     input {expression=&self:MaterialExpressionMultiply_50}
    }
    {
     inputName=Max
     input {expression=&self:MaterialExpressionMultiply_51}
    }
    {
     inputName=SceneDepth
     input {expression=&self:MaterialExpressionDivide_9}
    }
   ]
  }
  n15 = Custom(
   D=@MaterialExpressionMultiply_240
   O=@MaterialExpressionMultiply_241
   Min=@MaterialExpressionMultiply_50
   Max=@MaterialExpressionMultiply_51
   SceneDepth=@MaterialExpressionDivide_9
  ) {
   code="float a =  dot(D.xy, D.xy) - D.z * D.z;\r\nfloat b = (dot(O.xy, D.xy) - O.z * D.z) * 2;\r\nfloat c =  dot(O.xy, O.xy) - O.z * O.z;\r\n\r\nfloat d = abs( (b*b) - (4*a*c) );\r\nfloat t   = sqrt( d );\r\n\r\nfloat t0 = -b - t;\r\nfloat t1 = -b + t;\r\n\r\nt0 /= 2 * a;\r\nt1 /= 2 * a;\r\n\r\nfloat dist_to_top = max( 0, (Max - O.z) / D.z );\r\nfloat dist_to_bot = max( 0, (Min - O.z) / D.z );\r\n\r\nfloat furthestcap   = dist_to_top > dist_to_bot? Max: Min;\r\nfloat furthestcapXY = length( O.xy + max(dist_to_top, dist_to_bot) * D.xy );\r\n\r\n\r\nfloat entrytest = O.z + t0 * D.z;\r\nfloat exittest =  O.z + t1 * D.z;\r\n\r\nt0 = max( 0, t0);\r\nfloat t0_nocap = t0;\r\n\r\nif( entrytest < Min || entrytest > Max )\r\n{\r\n\tt0 = min(dist_to_top, dist_to_bot);\r\n}\r\n\r\nif( exittest  < Min || exittest  > Max )\r\n{\r\n\tt1 = max(dist_to_top, dist_to_bot);\r\n\t\r\n\tif(furthestcapXY > furthestcap ) t1 = 0;\r\n}\r\n\r\nt1 = min( t1, SceneDepth );\r\n\r\nif( d < 0) t1 = 0;\r\nfloat thickness = (t1 - t0);\r\n\r\nreturn float4(thickness, t0, t1, t0_nocap);\r\n\r\n\r\n"
   description="Cone Intersection (NaN Safe)"
   outputType=CMOT_Float4
   inputs [
    {
     inputName=D
     input {expression=&self:MaterialExpressionMultiply_240}
    }
    {
     inputName=O
     input {expression=&self:MaterialExpressionMultiply_241}
    }
    {
     inputName=Min
     input {expression=&self:MaterialExpressionMultiply_50}
    }
    {
     inputName=Max
     input {expression=&self:MaterialExpressionMultiply_51}
    }
    {
     inputName=SceneDepth
     input {expression=&self:MaterialExpressionDivide_9}
    }
   ]
  }
  n16 = MaterialFunctionCall(Float3=@n15) {materialFunction=&BreakOutFloat3Components}
  n17 = FunctionOutput(None=@n16.B) {outputName=T1, sortPriority=4}
  n18 = Multiply(A=Multiply(A=@n1) {desc=Ray Direction}) {desc=Ray Direction}
  n19 = Multiply(A=Multiply(A=@n9) {desc=Ray Origin}) {desc=Ray Origin}
  FunctionOutput(None=Add(A=Multiply(A=@n17, B=@n18), B=@n19)) {outputName=Ray Exit World Position, sortPriority=1}
  n20 = FunctionOutput(None=@n16.G) {outputName=T0, sortPriority=3}
  FunctionOutput(None=Add(A=Multiply(A=@n20, B=@n18), B=@n19)) {outputName=Ray Entry World Position}
  FunctionOutput(None=@n16.R) {outputName=Cone Thickness, sortPriority=2}
  n21 = Multiply(A=@MaterialExpressionMultiply_241) {desc=Local E}
  n22 = Multiply(A=Multiply(A=@MaterialExpressionMultiply_240) {desc=Local D})
  n23 = ComponentMask(None=@n15) {a=true}
  FunctionOutput(None=Add(A=Multiply(A=@n21), B=Multiply(A=@n22, B=@n23))) {outputName=Ray Entry Local Position, sortPriority=5}
 ]
}
