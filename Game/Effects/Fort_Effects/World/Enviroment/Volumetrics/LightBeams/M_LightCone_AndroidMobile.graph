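# Additive unlit material for volumetric light-beam cones, an Android mobile variant with optional lighthouse rotation.
# A ray-cone intersection feeds an analytic inverse-square scattering integral, guarded against NaNs.
# That result is multiplied by a Schlick phase term, edge fade, color, intensity and alpha.
# UV1 direction and red vertex color shape the mesh offset between two radii, while a static switch applies lighthouse rotation and collection-driven alpha.
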Material /Game/Effects/Fort_Effects/World/Enviroment/Volumetrics/LightBeams/M_LightCone_AndroidMobile
graph {
 properties {
  bEnableResponsiveAA=true
  bUseTranslucencyVertexFog=false
  blendMode=BLEND_Additive
  pixelDepthOffsetMode=PDOM_Legacy
  shadingModel=MSM_Unlit
  usedShadingModels=Unlit
 }
 nodes [
  Custom!(None) {
   code="return lerp(isfinite(input), input, 0);"
   desc="If this next round doesn't work, let's just try something like this"
   description=NaNProtection
   inputs [{inputName=input, input {expression=null}}]
  }
  n1 = VectorParameter() {parameterName=Light Vector, defaultValue=<rgba 0 1 0 1>}
  n2 = Constant3Vector() {constant=<rgba 0 0 1 0>}
  n3 = Divide(B=Max(A=ScalarParameter() {defaultValue=0.1, parameterName=Tan}) {constB=0.001})
  MaterialExpressionMaterialFunctionCall_1 = MaterialFunctionCall!(
   Ray Origin
   Ray Direction
   Cone Origin=TransformPosition(None=Constant3Vector()) {transformType=TRANSFORMPOSSOURCE_World}
   Cone Z Axis=StaticSwitchParameter(
    True=MaterialFunctionCall!(Postion to Transform=@n1.RGB, WPO) {materialFunction=/Game/Effects/Fort_Effects/World/Enviroment/Lighthouse/LighthouseRotationandControl}.Rotated Position
    False=Transform(None=@n2) {transformSourceType=TRANSFORMSOURCE_Local}
   ) {parameterName=Use LIghtHouse Rotation}
   Z Min-Max=AppendVector(A=ScalarParameter() {defaultValue=100, parameterName=ZMin}, B=ScalarParameter() {defaultValue=1000, parameterName=ZMax})
   XYZ Scale=MaterialFunctionCall(X=@n3, Y=@n3, Z=Constant() {r=1}) {materialFunction=&MakeFloat3}.Result
   CUV Vector 1
   CUV Vector 2
  ) {materialFunction=&/ConeIntersection}
  n4 = Constant()
  n5 = ScalarParameter() {defaultValue=100, parameterName=ZMin}
  n6 = ScalarParameter() {defaultValue=1, parameterName=LightShift}
  MaterialExpressionAdd_1 = Add(
   A=@MaterialExpressionMaterialFunctionCall_1.Local Ray Origin
   B=MaterialFunctionCall(X=@n4, Y=@n4, Z=Multiply(A=@n5, B=@n6)) {materialFunction=&MakeFloat3}.Result
  )
  MaterialExpressionScalarParameter_6 = ScalarParameter() {defaultValue=0.001, desc=Tiny offset to prevent huge brightness values, parameterName=f}
  Custom!(
   t0=@MaterialExpressionMaterialFunctionCall_1.T0
   t1=@MaterialExpressionMaterialFunctionCall_1.T0
   RayOrigin=@MaterialExpressionAdd_1
   RayDirection=@MaterialExpressionMaterialFunctionCall_1.T0
   f=@MaterialExpressionScalarParameter_6
  ) {
   code="float3 ro = RayOrigin;\r\nfloat3 rd = RayDirection;\r\n\r\nfloat a = dot(rd,rd);\r\nfloat b = dot(ro, rd);\r\nfloat c = dot(ro,ro);\r\n\r\nt0 = max( t0, 0.0f );\r\n\r\n//tiny offset param to avoid infinite brightness\r\nc+= f;\r\nfloat h2 = sqrt( (c) - (b*b) );\r\nfloat integral_t0 = ( atanFast( ( t0 + b ) / h2 ) ) / h2;\r\nfloat integral_t1 = ( atanFast( ( t1 + b ) / h2 ) ) / h2;\r\n\r\nreturn (integral_t1 - integral_t0);\r\n\r\n\r\n"
   description=Inverse Square Falloff
   outputType=CMOT_Float2
   inputs [
    {
     inputName=t0
     input {
      expression=&self:MaterialExpressionMaterialFunctionCall_1
      outputIndex=3
     }
    }
    {
     inputName=t1
     input {
      expression=&self:MaterialExpressionMaterialFunctionCall_1
      outputIndex=4
     }
    }
    {
     inputName=RayOrigin
     input {expression=&self:MaterialExpressionAdd_1}
    }
    {
     inputName=RayDirection
     input {
      expression=&self:MaterialExpressionMaterialFunctionCall_1
      outputIndex=7
     }
    }
    {
     inputName=f
     input {expression=&self:MaterialExpressionScalarParameter_6}
    }
   ]
  }
  n7 = Normalize(VectorInput=Subtract(A=TextureCoordinate() {coordinateIndex=1}) {constB=0.5})
  n8 = MaterialFunctionCall(Float2=@n7) {materialFunction=&BreakOutFloat2Components}
  MaterialFunctionCall(R=@n8.R, G=Constant(), B=@n8.R) {
   desc="transformLocalToWorld based on specific light beam mesh\r\nNeeded if used in HLOD"
   materialFunction=&AppendMany
  }
  n9 = VectorParameter() {parameterName=Color, defaultValue=<rgba 0.7201 2 3 1>}
  n10 = ScalarParameter() {defaultValue=1, parameterName=EmissiveIntensity}
  n11 = VectorParameter() {parameterName=Light Vector, defaultValue=<rgba 0 1 0 1>}
  n12 = Constant3Vector() {constant=<rgba 0 0 1 0>}
  MaterialExpressionDotProduct_2 = DotProduct(
   A=Multiply(A=CameraVectorWS()) {constB=-1}
   B=StaticSwitchParameter(
    True=MaterialFunctionCall(Postion to Transform=@n11.RGB) {materialFunction=&LighthouseRotationandControl}.Rotated Position
    False=Transform(None=@n12) {transformSourceType=TRANSFORMSOURCE_Local}
   ) {parameterName=Use LIghtHouse Rotation}
  )
  MaterialExpressionScalarParameter_8 = ScalarParameter() {defaultValue=0.75, parameterName=Anistropoy}
  n13 = Custom!(CosTheta=@MaterialExpressionDotProduct_2, k=@MaterialExpressionScalarParameter_8) {
   code="float Inner = (1 + k * CosTheta);\r\n\r\nInner = max(Inner, .001);\r\n\r\nreturn (1 - k * k) / (4 * PI * Inner * Inner);\r\n\r\n"
   description="SchlickPhase (NaN safe)"
   outputType=CMOT_Float1
   inputs [
    {
     inputName=CosTheta
     input {expression=&self:MaterialExpressionDotProduct_2}
    }
    {
     inputName=k
     input {expression=&self:MaterialExpressionScalarParameter_8}
    }
   ]
  }
  MaterialExpressionScalarParameter_0 = ScalarParameter() {defaultValue=0.001, desc=Tiny offset to prevent huge brightness values, parameterName=f}
  n14 = Custom!(
   t0=@MaterialExpressionMaterialFunctionCall_1.T0
   t1=@MaterialExpressionMaterialFunctionCall_1.T0
   RayOrigin=@MaterialExpressionAdd_1
   RayDirection=@MaterialExpressionMaterialFunctionCall_1.T0
   f=@MaterialExpressionScalarParameter_0
  ) {
   code="float3 ro = RayOrigin;\r\nfloat3 rd = RayDirection;\r\n\r\nfloat a = dot(rd,rd);\r\nfloat b = dot(ro, rd);\r\nfloat c = dot(ro,ro);\r\nfloat d =  c-b*b;\r\n//Do abs here to make sure d is never -1 for sqrt coming up\r\nd = abs(d);\r\n\r\nt0 = max( t0, 0.0f );\r\n\r\n//tiny offset param to avoid infinite brightness\r\nc+= f;\r\nfloat h2 = sqrt( d );\r\n//Make sure h2 is never exactly 0 to avoid NaN\r\nh2 = max( h2, .001);\r\n\r\nfloat integral_t0 = ( atanFast( ( t0 + b ) / h2 ) ) / h2;\r\nfloat integral_t1 = ( atanFast( ( t1 + b ) / h2 ) ) / h2;\r\n\r\nreturn (integral_t1 - integral_t0);\r\n\r\n\r\n\r\n"
   description="Inverse Square Falloff (With NaN Protection)"
   outputType=CMOT_Float2
   inputs [
    {
     inputName=t0
     input {
      expression=&self:MaterialExpressionMaterialFunctionCall_1
      outputIndex=3
     }
    }
    {
     inputName=t1
     input {
      expression=&self:MaterialExpressionMaterialFunctionCall_1
      outputIndex=4
     }
    }
    {
     inputName=RayOrigin
     input {expression=&self:MaterialExpressionAdd_1}
    }
    {
     inputName=RayDirection
     input {
      expression=&self:MaterialExpressionMaterialFunctionCall_1
      outputIndex=7
     }
    }
    {
     inputName=f
     input {expression=&self:MaterialExpressionScalarParameter_0}
    }
   ]
  }
  n15 = Multiply(A=@n13, B=ComponentMask(None=@n14) {r=true})
  n16 = DotProduct(
   A=Normalize(VectorInput=ComponentMask(None=@MaterialExpressionMaterialFunctionCall_1.Ray Entry Local Position) {g=true, r=true})
   B=Normalize(VectorInput=ComponentMask(None=@MaterialExpressionMaterialFunctionCall_1.Local Ray Origin) {g=true, r=true})
  )
  n17 = Add(
   A=ComponentMask(None=@MaterialExpressionMaterialFunctionCall_1.Local Ray Origin) {b=true}
   B=ScalarParameter() {defaultValue=100, parameterName=ZMin}
  )
  n18 = Max(A=ScalarParameter() {defaultValue=0.1, parameterName=Tan}) {constB=0.001}
  n19 = Divide(A=ComponentMask(None=@MaterialExpressionMaterialFunctionCall_1.Local Ray Origin) {g=true, r=true}, B=@n18)
  n20 = Divide(A=ScalarParameter() {defaultValue=2, parameterName=Fresnel Fade Distance}, B=@n18)
  n21 = Add(
   A=Subtract(
    A=Max(A=@n17) {constB=0}
    B=MaterialFunctionCall(Vector 2=@n19) {materialFunction=/Engine/Functions/Engine_MaterialFunctions02/Utility/VectorLength}.V2 Length
   )
   B=@n20
  )
  n22 = LinearInterpolate(A=Power(Base=Max(A=@n16) {constB=0.001}), Alpha=Clamp(None=Divide(A=@n21, B=@n20)))
  n23 = CollectionParameter() {collection=/Game/Effects/Fort_Effects/World/Enviroment/Lighthouse/LighthouseControl, parameterName=Alpha}
  n24 = ScalarParameter() {defaultValue=1, parameterName=Alpha}
  Multiply(
   A=Multiply(A=Multiply(A=@n9.RGB, B=@n10)) {constB=0.5}
   B=Multiply(A=Multiply(A=@n15, B=@n22), B=StaticSwitchParameter(True=@n23, False=@n24) {parameterName=Use LIghtHouse Rotation})
  )
  n25 = Transform(None=AppendVector(A=@n7, B=Constant())) {transformSourceType=TRANSFORMSOURCE_Local}
  n26 = MaterialFunctionCall(Postion to Transform=@n25) {materialFunction=&LighthouseRotationandControl}
  n27 = Multiply(
   A=LinearInterpolate(A=ScalarParameter() {defaultValue=10, parameterName=R1}, B=ScalarParameter() {defaultValue=150, parameterName=R2}, Alpha=VertexColor().R)
   B=Normalize(VectorInput=StaticSwitchParameter(True=@n26.Rotated Position, False=@n25) {parameterName=Use LIghtHouse Rotation})
  )
  StaticSwitchParameter(
   True=MaterialFunctionCall(WPO=@n27) {materialFunction=&LighthouseRotationandControl}.WPO
   False=@n27
  ) {parameterName=Use LIghtHouse Rotation}
 ]
}
